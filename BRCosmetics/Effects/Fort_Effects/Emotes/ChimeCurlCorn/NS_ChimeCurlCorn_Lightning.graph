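# Niagara lightning burst for the ChimeCurlCorn emote, with a red glow.
# Two near-identical emitters (reddish and white) each burst eight three-renderer mesh particles along a cylinder ring, oriented by the shape vector and accelerating outward.
# Curve-driven dynamic material parameters animate the meshes, and a scratch module takes camera-vector inputs.
# A growing, fading sprite supplies the glow, with time-of-day and front-end modifiers adjusting its alpha and the meshes' color attributes.
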NiagaraSystem /BRCosmetics/Effects/Fort_Effects/Emotes/ChimeCurlCorn/NS_ChimeCurlCorn_Lightning
dependencies {
 usedDataInterfaces [&NiagaraDataInterfaceCurve, &NiagaraDataInterfaceCamera]
 usedDynamicInputs []
 usedModules [&SystemState]
}
emitters {
 Lightning01 {
  emitterUpdateScript [
   &EmitterState(Life Cycle Mode=::NewEnumerator1, Loop Behavior=::NewEnumerator1, Loop Duration=1, UseLoopDelay=false)
   &SpawnBurst_Instantaneous(Spawn Count=8, Use Spawn Probability=false, UseLoopCountLimit=false)
  ]
  particleSpawnScript [
   &InitializeParticle(
    Color=<rgba 5 2.0342 1.4439 1.5>
    Lifetime=0.9
    Mesh Scale Mode=::NewEnumerator2
    Position Mode=::NewEnumerator1
    UsePositionOffset=false
    Write Material Random=true
   )
   &ShapeLocation(
    Cylinder Distribution=::NewEnumerator2
    Cylinder Height=&RandomRangeFloat(Minimum=0, Maximum=0)
    Cylinder Height Midpoint=0
    Cylinder Radius=&RandomRangeFloat(Minimum=20, Maximum=20)
    MaskSpawnGroup=false
    Non Uniform Scale=<1.2 1 1>
    Offset Mode=::NewEnumerator0
    Rotation Angle=0.2
    Rotation Axis=<0 0 1>
    Rotation Coordinate Space=::World
    Rotation Mode=::NewEnumerator1
    Shape Primitive=::NewEnumerator1
    U Position=&ReturnNormalizedExecIndex(Normalized Index Scale=0.77, Normalized Index Includes One=false)
    UseOwnerScale=true
   )
   &InitialMeshOrientation(Mesh Orientation Mode=::NewEnumerator2, Orientation Vector=$Particles.ShapeLocation.ShapeVector, Use Rotation Vector=false)
   &CameraOffset(Camera Offset Amount=-20)
   &AddVelocity(Module.Use Falloff Distance=false, Override Seed=false, Velocity Mode=::NewEnumerator1, Velocity Origin=<0 0 0>, Velocity Speed=35)
  ]
  particleUpdateScript [
   &ParticleState()
   &ScaleMeshSize(Scale Factor=&Multiply_VectorByFloat(Vector=<1 1 1>, Float=1))
   &DynamicMaterialParameters(
    Index 0 Param 1=&FloatFromCurve(FloatCurve=DataInterface() {propertyValues {Curve {defaultValue=0, keys [{tangentMode=RCTM_User, time=0, value=0}, {tangentMode=RCTM_User, time=1, value=0.8}]}}})
    Index 0 Param 2=50
    Index 0 Param 4=&FloatFromCurve(FloatCurve=DataInterface() {propertyValues {Curve {defaultValue=1, keys [{tangentMode=RCTM_User, time=0, value=0}, {tangentMode=RCTM_User, time=0.3, value=1}]}}})
    Index 1 Param 1=&FloatFromCurve(FloatCurve=DataInterface() {propertyValues {Curve {defaultValue=0, keys [{tangentMode=RCTM_User, time=0, value=0}, {tangentMode=RCTM_User, time=1, value=0.8}]}}})
    Index 1 Param 2=65
    Index 1 Param 4=&FloatFromCurve(FloatCurve=DataInterface() {propertyValues {Curve {defaultValue=1, keys [{tangentMode=RCTM_User, time=0, value=0}, {tangentMode=RCTM_User, time=0.3, value=1}]}}})
    Module.Index0Float0Write=true
    Module.Index0Float1Write=true
    Module.Index0Float2Write=true
    Module.Index0Float3Write=true
    Module.Index1Float0Write=true
    Module.Index1Float1Write=true
    Module.Index1Float2Write=true
    Module.Index1Float3Write=true
    Write Parameter Index 1=::NewEnumerator1
   )
   &FN_Sprite_Attributes(
    Color=&MakeLinearColorFromVectorAndFloat(
     "Vector (RGB)"=&Multiply_VectorByFloat(Vector=&MakeVectorFromLinearColorRGB(Color=<rgba 1 0 0 1>), Float=1)
     "Float (Alpha)"=&Multiply_Float(
      A=&FNTimeOfDayScalar(Night Scale=0.4, Morning Scale=0.7, Evening Scale=0.7)
      B=&FN_FrontEndFloatModifier(Gameplay=1, Front End=0.3)
     )
    )
    Create Color Attribute=true
    Write to Module Namespace=true
   )
   &FN_Sprite_Attributes(
    Color=&MakeLinearColorFromVectorAndFloat(
     "Vector (RGB)"=&Multiply_VectorByFloat(Vector=&MakeVectorFromLinearColorRGB(Color=<rgba 1 0 0 1>), Float=1.5)
     "Float (Alpha)"=&Multiply_Float(A=10, B=1)
    )
    Create Color Attribute=true
    Write to Module Namespace=true
   ) {module=FN_Sprite_Attributes001}
   &ScaleVelocity(
    Coordinate Space:ENiagaraCoordinateSpace=::World
    Velocity Scale=&Multiply_VectorByFloat(
     Vector=<1 1 1>
     Float=&FloatFromCurve(FloatCurve=DataInterface() {propertyValues {Curve {keys [
          {arriveTangent=1.9405, interpMode=RCIM_Cubic, leaveTangent=1.9405, tangentMode=RCTM_User, time=0, value=0}
          {arriveTangent=5, interpMode=RCIM_Cubic, leaveTangent=5, time=0.507, value=0.8665}
          <at 1 5>
         ]}}})
    )
   )
   &SolveForcesAndVelocity(Clamp Velocity=false, Limit Acceleration=false)
   &self:ScratchModule(
    New:Quat4f=&MakeQuaternion(
     Quaternion From=3
     Axis=<1 0 0>
     Angle=857.7451
     From=<0 0 1>
     To=&ReturnCameraVector(
      Camera Vectors=::NewEnumerator1
      Coordinate Space=::Simulation
      CameraQuery=DataInterface() {propertyValues {bRequireCurrentFrameData=false}}
     )
     Angle Type=::NewEnumerator0
     Normalize Vectors=false
    )
    Vector:Vector3f=&ReturnCameraVector(
     Camera Vectors=::NewEnumerator1
     Coordinate Space=::Simulation
     CameraQuery=DataInterface() {propertyValues {bRequireCurrentFrameData=false}}
    )
   ) {module=ScratchModule}
   /SetVariables(Particles.T_Position:NiagaraPosition=&AddVectorToPosition(Position=$StackContext.T_Position, Vector=Linked() {linkedVariable {name=Engine.Emitter.SimulationPosition, type=NiagaraPosition}})) {isSetParameters=true}
  ]
  rendererClasses [&NiagaraMeshRendererProperties]
  renderers [
   {rendererClass=&NiagaraMeshRendererProperties, rendererIndex=0}
   {rendererClass=&NiagaraMeshRendererProperties, rendererIndex=1}
   {rendererClass=&NiagaraMeshRendererProperties, rendererIndex=2}
  ]
 }
 Lightning02 {
  emitterUpdateScript [
   &EmitterState(Life Cycle Mode=::NewEnumerator1, Loop Behavior=::NewEnumerator1, Loop Duration=1, UseLoopDelay=false)
   &SpawnBurst_Instantaneous(Spawn Count=8, Use Spawn Probability=false, UseLoopCountLimit=false)
  ]
  particleSpawnScript [
   &InitializeParticle(
    Color=<rgba 10 10 10 1.5>
    Lifetime=0.8
    Mesh Scale Max=<2 1.5 1.5>
    Mesh Scale Mode=::NewEnumerator2
    Position Mode=::NewEnumerator1
    UsePositionOffset=false
    Write Material Random=true
   )
   &ShapeLocation(
    Cylinder Distribution=::NewEnumerator2
    Cylinder Height=&RandomRangeFloat(Minimum=0, Maximum=0)
    Cylinder Height Midpoint=0
    Cylinder Radius=&RandomRangeFloat(Minimum=15, Maximum=20)
    MaskSpawnGroup=false
    Non Uniform Scale=<1.2 1 1>
    Offset Mode=::NewEnumerator0
    Rotation Angle=0.16
    Rotation Axis=<0 0 1>
    Rotation Coordinate Space=::World
    Rotation Mode=::NewEnumerator1
    Shape Primitive=::NewEnumerator1
    U Position=&ReturnNormalizedExecIndex(Normalized Index Scale=0.77, Normalized Index Includes One=false)
    UseOwnerScale=true
   )
   &InitialMeshOrientation(Mesh Orientation Mode=::NewEnumerator2, Orientation Vector=$Particles.ShapeLocation.ShapeVector, Use Rotation Vector=false)
   &CameraOffset(Camera Offset Amount=-20)
   &AddVelocity(Module.Use Falloff Distance=false, Override Seed=false, Velocity Mode=::NewEnumerator1, Velocity Origin=<0 0 0>, Velocity Speed=35)
  ]
  particleUpdateScript [
   &ParticleState()
   &ScaleMeshSize(Scale Factor=&Multiply_VectorByFloat(Vector=<1 1 1>, Float=1))
   &DynamicMaterialParameters(
    Index 0 Param 1=&FloatFromCurve(FloatCurve=DataInterface() {propertyValues {Curve {defaultValue=0, keys [{tangentMode=RCTM_User, time=0, value=0}, {tangentMode=RCTM_User, time=1, value=0.8}]}}})
    Index 0 Param 2=50
    Index 0 Param 4=&FloatFromCurve(FloatCurve=DataInterface() {propertyValues {Curve {defaultValue=1, keys [{tangentMode=RCTM_User, time=0, value=0}, {tangentMode=RCTM_User, time=0.3, value=1}]}}})
    Index 1 Param 1=&FloatFromCurve(FloatCurve=DataInterface() {propertyValues {Curve {defaultValue=0, keys [{tangentMode=RCTM_User, time=0, value=0}, {tangentMode=RCTM_User, time=1, value=0.8}]}}})
    Index 1 Param 2=60
    Index 1 Param 4=&FloatFromCurve(FloatCurve=DataInterface() {propertyValues {Curve {defaultValue=1, keys [{tangentMode=RCTM_User, time=0, value=0}, {tangentMode=RCTM_User, time=0.3, value=1}]}}})
    Module.Index0Float0Write=true
    Module.Index0Float1Write=true
    Module.Index0Float2Write=true
    Module.Index0Float3Write=true
    Module.Index1Float0Write=true
    Module.Index1Float1Write=true
    Module.Index1Float2Write=true
    Module.Index1Float3Write=true
    Write Parameter Index 1=::NewEnumerator1
   )
   &FN_Sprite_Attributes(
    Color=&MakeLinearColorFromVectorAndFloat(
     "Vector (RGB)"=&Multiply_VectorByFloat(Vector=&MakeVectorFromLinearColorRGB(Color=<rgba 1 0 0 1>), Float=1)
     "Float (Alpha)"=&Multiply_Float(
      A=&FNTimeOfDayScalar(Night Scale=0.4, Morning Scale=0.7, Evening Scale=0.7)
      B=&FN_FrontEndFloatModifier(Gameplay=1, Front End=0.3)
     )
    )
    Create Color Attribute=true
    Write to Module Namespace=true
   )
   &FN_Sprite_Attributes(
    Color=&MakeLinearColorFromVectorAndFloat(
     "Vector (RGB)"=&Multiply_VectorByFloat(Vector=&MakeVectorFromLinearColorRGB(Color=<rgba 1 0 0 1>), Float=1.5)
     "Float (Alpha)"=&Multiply_Float(A=10, B=1)
    )
    Create Color Attribute=true
    Write to Module Namespace=true
   ) {module=FN_Sprite_Attributes001}
   &ScaleVelocity(
    Coordinate Space:ENiagaraCoordinateSpace=::World
    Velocity Scale=&Multiply_VectorByFloat(
     Vector=<1 1 1>
     Float=&FloatFromCurve(FloatCurve=DataInterface() {propertyValues {Curve {keys [
          {arriveTangent=1.9405, interpMode=RCIM_Cubic, leaveTangent=1.9405, tangentMode=RCTM_User, time=0, value=0}
          {arriveTangent=5, interpMode=RCIM_Cubic, leaveTangent=5, time=0.507, value=0.8665}
          <at 1 5>
         ]}}})
    )
   )
   &SolveForcesAndVelocity(Clamp Velocity=false, Limit Acceleration=false)
   &self:ScratchModule(
    New:Quat4f=&MakeQuaternion(
     Quaternion From=3
     Axis=<1 0 0>
     Angle=857.7451
     From=<0 0 1>
     To=&ReturnCameraVector(
      Camera Vectors=::NewEnumerator1
      Coordinate Space=::Simulation
      CameraQuery=DataInterface() {propertyValues {bRequireCurrentFrameData=false}}
     )
     Angle Type=::NewEnumerator0
     Normalize Vectors=false
    )
    Vector:Vector3f=&ReturnCameraVector(
     Camera Vectors=::NewEnumerator1
     Coordinate Space=::Simulation
     CameraQuery=DataInterface() {propertyValues {bRequireCurrentFrameData=false}}
    )
   ) {module=ScratchModule}
   /SetVariables(Particles.T_Position:NiagaraPosition=&AddVectorToPosition(Position=$StackContext.T_Position, Vector=Linked() {linkedVariable {name=Engine.Emitter.SimulationPosition, type=NiagaraPosition}})) {isSetParameters=true}
  ]
  rendererClasses [&NiagaraMeshRendererProperties]
  renderers [
   {rendererClass=&NiagaraMeshRendererProperties, rendererIndex=0}
   {rendererClass=&NiagaraMeshRendererProperties, rendererIndex=1}
   {rendererClass=&NiagaraMeshRendererProperties, rendererIndex=2}
  ]
 }
 Minimal {
  emitterUpdateScript [
   &EmitterState()
   &SpawnBurst_Instantaneous(Use Spawn Probability=false, UseLoopCountLimit=false)
  ]
  particleSpawnScript [
   &InitializeParticle(
    Color=<rgba 0.5 0.0075 0.0075 1>
    Lifetime=0.8
    Position Offset=<0 0 15>
    Sprite Size Mode=::NewEnumerator3
    Uniform Sprite Size=200
    UsePositionOffset=true
    Write Material Random=true
   )
   &DynamicMaterialParameters(
    Index 0 Param 1=3
    Index 0 Param 2=50
    Module.Index0Float0Write=true
    Module.Index0Float1Write=true
    Module.Index0Float2Write=true
    Module.Index0Float3Write=true
   )
   &CameraOffset(Camera Offset Amount=&FN_FrontEndFloatModifier(Gameplay=-30, Front End=-20))
  ]
  particleUpdateScript [
   &ParticleState()
   &ScaleSpriteSize(
    Uniform Curve Scale=3
    Uniform Curve Sprite Scale=DataInterface() {propertyValues {Curve {keys [
        {tangentMode=RCTM_User, time=0, value=0}
        {arriveTangent=1, interpMode=RCIM_Cubic, leaveTangent=1, time=0.2853, value=0.632}
        {tangentMode=RCTM_User, time=1, value=1}
       ]}}}
   )
   &ScaleColor(
    Scale Alpha=&FloatFromCurve(
     FloatCurve=DataInterface() {propertyValues {Curve {keys [<at 0 1>,<at 1 0>]}}}
     Scale Curve=&Multiply_Float(
      A=&FN_FrontEndFloatModifier(Gameplay=0.6, Front End=0.2)
      B=&FNTimeOfDayScalar(Night Scale=0.2, Morning Scale=0.5, Evening Scale=0.5)
     )
    )
    ScaleA=true
    ScaleRGB=true
   )
  ]
  rendererClasses [&NiagaraSpriteRendererProperties]
  renderers [{rendererClass=&NiagaraSpriteRendererProperties}]
 }
}
summary {
 systemName=NS_ChimeCurlCorn_Lightning
 emitters [
  {emitterName=Lightning01, rendererClasses [&NiagaraMeshRendererProperties]}
  {emitterName=Lightning02, rendererClasses [&NiagaraMeshRendererProperties]}
  {emitterName=Minimal, rendererClasses [&NiagaraSpriteRendererProperties]}
 ]
 userVariables [
  User.bIsFrontEnd {type=NiagaraBool}
  User.bIsFrontEndPreview {type=NiagaraBool}
 ]
}
